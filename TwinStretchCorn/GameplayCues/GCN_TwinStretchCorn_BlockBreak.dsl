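; Burst gameplay cue for the TwinStretchCorn weapon that positions the Niagara effect, evidently for its block breaking.
; On burst, it casts the target to a Fort pawn and gets its current weapon as the TwinStretchCorn weapon.
; It reads the world transform of the "offset" bone on the spawned shield's shards mesh.
; The first spawned particle component moves to that location and rotation, aligning the effect with the shield.
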(event EventOnBurst(Niagara)
  (params (MyTarget :type Actor%20Object%20Reference) (Parameters :type Gameplay%20Cue%20Parameters%20Structure%20%28by%20ref%29) (ParticleComponents :type Array%20of%20Niagara%20Particle%20System%20Component%20Object%20References) (AudioComponents :type Array%20of%20Audio%20Component%20Object%20References) (BurstCameraShakeInstance :type Camera%20Shake%20Base%20Object%20Reference) (BurstDecalInstance :type Decal%20Actor%20Object%20Reference))
  (bind _asfort_pawn (Utilities|Casting|CastToFortPawn MyTarget))
  (bind _asb_weap_twin_stretch_corn (Utilities|Casting|CastToB_Weap_TwinStretchCorn (Variables|/Script/FortniteGame.FortPawn|GetCurrentWeapon _asfort_pawn)))
  (bind (_returnvalue_location _returnvalue_rotation) (Math|Transform|BreakTransform (Transformation|GetBoneTransform (Variables|/TwinStretchCorn/Gameplay/Abilities/Block/BGA_TwinStretchCorn_BlockVolume.BGA_TwinStretchCorn_BlockVolume_C|GetShardsMesh (Variables|/TwinStretchCorn/Gameplay/Actors/B_Weap_TwinStretchCorn.B_Weap_TwinStretchCorn_C|GetSpawnedShield _asb_weap_twin_stretch_corn)) "offset" "RTS_World")))
  (Transformation|SetWorldLocationAndRotation (Utilities|Array|Get(acopy) ParticleComponents) _returnvalue_location _returnvalue_rotation false false))
